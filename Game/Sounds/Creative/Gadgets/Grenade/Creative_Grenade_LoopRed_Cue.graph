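# Sound cue for the Creative grenade gadget, named as its red loop variant.
# A mixer with one connected input and one empty slot plays the reused MoveTool_Granted_02 wave at half pitch and reduced volume.
# The wave player does not enable looping, despite the cue's name.
# It uses the Creative gadgets sound class and borrows the MoveTool attenuation.
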SoundCue /Game/Sounds/Creative/Gadgets/Grenade/Creative_Grenade_LoopRed_Cue
root=Mixer_0
nodes [
 {id=Mixer_0, children [WavePlayer_1,null], properties {inputVolume [1,1]}}
 {id=WavePlayer_1, properties {soundWaveAssetPtr=/Game/Sounds/Creative/Gadgets/MoveTool/MoveTool_Granted_02}}
]
properties {
 attenuationSettings=/Game/Sounds/Creative/Attenuations/Fort_Gadget_MoveTool_Att
 maxDistance=4000
 pitchMultiplier=0.5
 soundClassObject=&Fort_Gadgets_Creative
 volumeMultiplier=0.7
}
