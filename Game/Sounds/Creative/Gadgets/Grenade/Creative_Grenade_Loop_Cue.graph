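# Sound cue for the Creative grenade gadget's loop, built from reused MoveTool gadget sounds.
# A mixer plays a one-shot granted sound immediately, alongside a looping selected-loop wave that starts after a short fixed delay.
# It routes through the Creative gadgets sound class and uses the MoveTool attenuation asset, with reduced overall volume.
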SoundCue /Game/Sounds/Creative/Gadgets/Grenade/Creative_Grenade_Loop_Cue
root=Mixer_0
nodes [
 {id=Delay_0, children [WavePlayer_0], properties {delayMax=0.1, delayMin=0.1}}
 {id=Mixer_0, children [WavePlayer_1,Delay_0], properties {inputVolume [1,1]}}
 {id=WavePlayer_0, properties {bLooping=true, soundWaveAssetPtr=/Game/Sounds/Creative/Gadgets/MoveTool/MoveTool_SelectedLoop_02}}
 {id=WavePlayer_1, properties {soundWaveAssetPtr=/Game/Sounds/Creative/Gadgets/MoveTool/MoveTool_Granted_02}}
]
properties {
 attenuationSettings=/Game/Sounds/Creative/Attenuations/Fort_Gadget_MoveTool_Att
 maxDistance=4000
 soundClassObject=&Fort_Gadgets_Creative
 volumeMultiplier=0.7
}
